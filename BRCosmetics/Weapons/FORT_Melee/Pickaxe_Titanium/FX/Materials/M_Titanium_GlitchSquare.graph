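# Translucent unlit Niagara sprite material for a glitching square effect on the Titanium pickaxe.
# A panned lightning texture, sampled in camera space, distorts the sprite's UVs along a vector, scaled by a dynamic parameter.
# The horizontal coordinate is hard-thresholded against dynamic Color Aberration and Mask values.
# These cuts switch between aberration-tinted and plain particle color, fed to an emissive adjustment function, and mask particle alpha.
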Material /BRCosmetics/Weapons/FORT_Melee/Pickaxe_Titanium/FX/Materials/M_Titanium_GlitchSquare
graph {
 properties {bUsedWithNiagaraSprites=true, blendMode=BLEND_Translucent, pixelDepthOffsetMode=PDOM_Legacy, shadingModel=MSM_Unlit, usedShadingModels=Unlit}
 nodes [
  n1 = WorldPosition()
  n2 = Multiply(
   A=ComponentMask(None=TransformPosition(None=@n1.XYZ) {transformSourceType=TRANSFORMPOSSOURCE_World, transformType=TRANSFORMPOSSOURCE_Camera}) {g=true, r=true}
   B=ScalarParameter() {defaultValue=0.01, parameterName=Distortion Tiling, sliderMax=5}
  )
  AppendVector(A=@n2, B=Constant() {r=0.5})
  n3 = VectorParameter() {parameterName=Aberration Color, defaultValue=<rgba 1 0 0 1>}
  n4 = ParticleColor()
  n5 = TextureCoordinate()
  n6 = VectorParameter() {parameterName=Distortion Vector, defaultValue=<rgba 1 0 0 1>}
  n7 = TextureSample(
   UVs=Panner(Coordinate=@n2)
   Tex=TextureObjectParameter() {parameterName=Lightning Texture, texture=/BRCosmetics/Weapons/FORT_Melee/Pickaxe_Titanium/FX/Texture/T_Titanium_LightningV4}
  ) {texture=&T_Titanium_LightningV4}
  n8 = Multiply(
   A=ComponentMask(None=Multiply(A=@n6.RGB, B=@n7.B)) {g=true, r=true}
   B=DynamicParameter() {defaultValue=<rgba 0.2 0.1 0.3 1>, paramNames [Color Aberration, Mask, Distortion, Param4]}.Distortion
  )
  n9 = ComponentMask(None=Frac(None=Add(A=@n5, B=@n8))) {r=true}
  n10 = DynamicParameter() {defaultValue=<rgba 0.2 0.1 0.3 1>, paramNames [Color Aberration, Mask, Distortion, Param4]}
  n11 = Constant() {r=1}
  n12 = Constant()
  MaterialFunctionCall(Emissive=LinearInterpolate(A=Multiply(A=@n3.RGB, B=@n4.RGB), B=@n4.RGB, Alpha=If(A=@n9, B=@n10.Color Aberration, "A > B"=@n11, "A < B"=@n12))) {materialFunction=&MF_FX_EmissiveAdjustments}
  Multiply(A=@n4.A, B=If(A=@n9, B=@n10.Mask, "A > B"=Constant() {r=1}, "A < B"=Constant()))
  TextureCoordinate()
  Transform(None=@n6.RGB) {transformSourceType=TRANSFORMSOURCE_View, transformType=TRANSFORM_Tangent}
 ]
}
