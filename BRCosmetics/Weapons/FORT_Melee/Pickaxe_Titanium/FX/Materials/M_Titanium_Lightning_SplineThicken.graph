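# Translucent unlit lightning material for Titanium pickaxe Niagara mesh particles, with SplineThicken controlling ribbon width.
# UVs scroll in time steps, pinch toward the ends and are warped by a camera-space noise lookup.
# Three offset texture samples add tinted color fringing; a Glow static switch instead uses a flat color and the green channel.
# Particle dynamic parameters drive width, U scale, distortion, aberration and a sine swipe mask on opacity.
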Material /BRCosmetics/Weapons/FORT_Melee/Pickaxe_Titanium/FX/Materials/M_Titanium_Lightning_SplineThicken
graph {
 properties {
  bUsedWithNiagaraMeshParticles=true
  blendMode=BLEND_Translucent
  pixelDepthOffsetMode=PDOM_Legacy
  shadingModel=MSM_Unlit
  usedShadingModels=Unlit
 }
 nodes [
  n1 = DynamicParameter() {defaultValue=<rgba 0 1 1 0>, paramNames [Swipe, Scale U, Width, Distortion]}
  MaterialFunctionCall(WidthBase=@n1.Width, WidthTip=@n1.Width, Expand U or V UV Channel=StaticBool() {value=true}) {materialFunction=&SplineThicken}
  n2 = VectorParameter() {parameterName=Glow Color, defaultValue=<rgba 5 4.434 3.2995 1>}
  n3 = ParticleColor()
  n4 = NamedRerouteUsage()
  n5 = TextureObjectParameter() {parameterName=Lightning Parameter, texture=/BRCosmetics/Weapons/FORT_Melee/Pickaxe_Titanium/FX/Texture/T_Titanium_LightningV4}
  n6 = TextureSample(UVs=@n4, Tex=@n5)
  n7 = DynamicParameter() {parameterIndex=1, defaultValue=<rgba 0 1 1 0>, paramNames [Color Aberration, Param2, Param3, Param4]}
  n8 = TextureSample(UVs=Add(A=@n4, B=@n7.Color Aberration), Tex=@n5)
  n9 = TextureSample(UVs=Add(A=@n4, B=Multiply(A=@n7.Color Aberration) {constB=-1}), Tex=@n5)
  n10 = Max(A=Max(A=@n6.R, B=@n8.R), B=@n9.R)
  n11 = Multiply(A=@n3.RGB, B=@n10)
  n12 = Add(
   A=Multiply(A=@n8.R, B=VectorParameter() {parameterName=Color Aberration 1, defaultValue=<rgba 1 0 0 1>}.RGB)
   B=Multiply(A=@n9.R, B=VectorParameter() {parameterName=Color Aberration 2, defaultValue=<rgba 0 1 0 1>}.RGB)
  )
  n13 = Add(A=Multiply(A=Subtract(A=@n11, B=@n12), B=@n6.R), B=@n12)
  MaterialFunctionCall(Emissive=StaticSwitchParameter(True=@n2.RGB, False=Multiply(A=@n3.RGB, B=@n13)) {parameterName=Glow}) {materialFunction=&MF_FX_EmissiveAdjustments}
  n14 = Multiply(A=@n2.A, B=@n6.G)
  n15 = ComponentMask(None=TextureCoordinate()) {g=true}
  n16 = DynamicParameter() {defaultValue=<rgba 0 1 1 0>, paramNames [Swipe, Scale U, Width, Distortion]}
  n17 = Sine(None=Multiply(A=Add(A=@n15, B=@n16.Swipe) {constB=0}) {constB=0.5})
  Multiply(A=Multiply(A=@n3.A, B=StaticSwitchParameter(True=@n14, False=@n10) {parameterName=Glow}), B=Saturate(None=Multiply(A=@n17) {constB=2}))
  n18 = TransformPosition(None=WorldPosition().XYZ) {transformSourceType=TRANSFORMPOSSOURCE_World, transformType=TRANSFORMPOSSOURCE_Camera}
  n19 = AppendVector(
   A=Multiply(A=ComponentMask(None=@n18) {g=true}, B=ScalarParameter() {defaultValue=0.1, parameterName=Distortion Tiling, sliderMax=5})
   B=Constant() {r=0.5}
  )
  n20 = Multiply(
   A=Transform(None=VectorParameter() {parameterName=Distortion Vector, defaultValue=<rgba 0 1 0 1>}.RGB) {transformSourceType=TRANSFORMSOURCE_View, transformType=TRANSFORM_Tangent}
   B=TextureSample(
    UVs=Panner(Coordinate=@n19)
    Tex=TextureObjectParameter() {parameterName=Lightning Texture, texture=&T_Titanium_LightningV4}
   ) {texture=&T_Titanium_LightningV4}.B
  )
  NamedRerouteDeclaration(None=Multiply(
    A=ComponentMask(None=@n20) {g=true, r=true}
    B=DynamicParameter() {defaultValue=<rgba 0 1 1 0>, paramNames [Swipe, Scale U, Width, Distortion]}.Distortion
   )) {name=Distortion, nodeColor=<rgba 0 1 0.7891 1>}
  n21 = MaterialFunctionCall(XY=TextureCoordinate()) {materialFunction=&Swizzle}
  n22 = ScalarParameter() {defaultValue=5, parameterName=Steps, sliderMax=50}
  n23 = Multiply(A=Multiply(A=Time(), B=ScalarParameter() {defaultValue=1, parameterName=Speed X, sliderMax=10}), B=@n22)
  n24 = AppendVector(A=Divide(A=Ceil(None=@n23), B=@n22), B=Constant())
  n25 = DynamicParameter() {defaultValue=<rgba 0 1 1 0>, paramNames [Swipe, Scale U, Width, Distortion]}
  n26 = Constant() {r=1}
  n27 = SphereMask(A=ComponentMask(None=TextureCoordinate()) {g=true}, B=Constant() {r=0.5}) {attenuationRadius=0.55, hardnessPercent=10}
  n28 = MaterialFunctionCall(
   UVs=Multiply(A=Add(A=@n21.YX, B=@n24), B=AppendVector(A=@n25.Scale U, B=@n26))
   Texture Scale=AppendVector(A=Constant() {r=1}, B=Saturate(None=@n27))
  ) {materialFunction=&ScaleUVsByCenter}
  n29 = ComponentMask(None=@n28.UVs) {r=true}
  n30 = Saturate(None=ComponentMask(None=@n28.UVs) {g=true})
  NamedRerouteDeclaration(None=Add(A=AppendVector(A=@n29, B=@n30), B=NamedRerouteUsage())) {name=Distorted UV, nodeColor=<rgba 0 1 0.9766 1>}
 ]
}
